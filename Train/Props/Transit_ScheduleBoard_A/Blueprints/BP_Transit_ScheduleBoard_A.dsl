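; Transit schedule board prop Blueprint that changes its display material when a train arrives at its station.
; When the arrival interval event reports a time near zero, it switches a custom primitive data float on the static mesh, then resets it after a ten-second delay.
; It supplies that same mesh through GetTimerMesh and leaves construction to its parent class.
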(event EventOnArrivalTimeIntervalReachedForStation
  (params (TImeUntilArrival :type Float%20%28double-precision%29))
  (bind _staticmesh (Variables|Default|GetStaticMesh))
  (if (Math|Float|NearlyEqual(Float) TImeUntilArrival)
    (Rendering|Material|SetCustomPrimitiveDataFloat _staticmesh 1 1.0)
    (Utilities|FlowControl|Delay 1e+01)
    (Rendering|Material|SetCustomPrimitiveDataFloat _staticmesh 1 0.0)))

(fn GetTimerMesh
  (return (Variables|Default|GetStaticMesh)))

(fn ConstructionScript
  (|CallParentFunction))
